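# Physics asset for the F_MED_RimeSkate_Plow female player cosmetic body, covering a five-piece belt plus the pelvis, spine_02 and thighs.
# Capsules approximate the body and most belt pieces, with a box on belt_02.
# Two tightly limited belt chains (01-02 and 06-05) and belt_03 hang from the pelvis.
# Angularly free constraints tie belt_03 to both chain ends, closing the belt into loops, and likewise join spine_02 to the pelvis.
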PhysicsAsset /BRCosmetics/Characters/Player/Female/Medium/Bodies/F_MED_RimeSkate_Plow/Meshes/F_MED_RimeSkate_Plow_Physics_FE
bodies {
 belt_01 [{
   shapeType=Capsule
   length=8.2463
   radius=3.1064
   shapeName=belt_01_capsule
   center=<6.44 -0.3043 0.1926>
   rotation=<rot -0.0001 -88.7237 -86.4323>
  }]
 belt_02 [{
   shapeType=Box
   extentX=15.942
   extentY=3.3742
   extentZ=8.382
   shapeName=belt_02_box
   center=<6.8065 0.6574 0.9942>
   rotation=<rot -5.2744 16.743 -1.5841>
  }]
 belt_03 [{
   shapeType=Capsule
   length=0.6347
   radius=7.6322
   shapeName=belt_05_capsule
   center=<19.0425 -2.0505 2.1435>
   rotation=<rot -16.3284 177.2667 -71.5495>
  }]
 belt_05 [{
   shapeType=Capsule
   length=4.9947
   radius=3.2723
   shapeName=belt_05_capsule
   center=<-3.5841 1.0269 0.1469>
   rotation=<rot -4.361 -118.0822 76.3515>
  }]
 belt_06 [{
   shapeType=Capsule
   length=4.1472
   radius=3.0509
   shapeName=belt_06_capsule
   center=<4.9027 1.2824 0.1095>
   rotation=<rot 11.133 -80.1786 94.9461>
  }]
 pelvis [
  {shapeType=Capsule, length=8.8906, radius=6.1895, center=<1.8148 -1.2183 -1.4992>}
  {shapeType=Capsule, length=18.3653, radius=6.3701, shapeName=pelvis_capsule, center=<-6.0267 -1.9353 -4.9772>}
 ]
 spine_02 [{shapeType=Capsule, length=9.9856, radius=9.161, shapeName=spine_02_capsule, center=<1.7322 0 0>, rotation=<rot 90 0 0>}]
 thigh_l [{shapeType=Capsule, length=25.1365, radius=6.3121, center=<-21.4737 0.2745 -0.5655>, rotation=<rot 0.2362 -92.4203 -87.5235>}]
 thigh_r [{shapeType=Capsule, length=22.7576, radius=8.6909, center=<21.4737 -0.2745 0.5655>, rotation=<rot 0.2362 -92.4203 92.4765>}]
}
constraints [
 {bone1Name=thigh_l, bone2Name=pelvis, swing1LimitDegrees=45, swing2LimitDegrees=45, twistLimitDegrees=45}
 {bone1Name=thigh_r, bone2Name=pelvis, swing1LimitDegrees=45, swing2LimitDegrees=45, twistLimitDegrees=45}
 {bone1Name=belt_01, bone2Name=pelvis, swing1LimitDegrees=40, swing2LimitDegrees=20, twistLimitDegrees=5}
 {bone1Name=belt_02, bone2Name=belt_01, swing1LimitDegrees=30, swing2LimitDegrees=20, twistLimitDegrees=5}
 {bone1Name=belt_06, bone2Name=pelvis, swing1LimitDegrees=5, swing2LimitDegrees=20, twistLimitDegrees=0}
 {bone1Name=belt_05, bone2Name=belt_06, swing1LimitDegrees=5, swing2LimitDegrees=20, twistLimitDegrees=0}
 {bone1Name=belt_03, bone2Name=pelvis, swing1LimitDegrees=5, swing2LimitDegrees=20, twistLimitDegrees=10}
 {
  bone1Name=belt_03
  bone2Name=belt_05
  swing1LimitDegrees=45
  swing1Motion=Free
  swing2LimitDegrees=45
  swing2Motion=Free
  twistLimitDegrees=45
  twistMotion=Free
 }
 {
  bone1Name=belt_03
  bone2Name=belt_02
  swing1LimitDegrees=45
  swing1Motion=Free
  swing2LimitDegrees=45
  swing2Motion=Free
  twistLimitDegrees=45
  twistMotion=Free
 }
 {
  bone1Name=spine_02
  bone2Name=pelvis
  swing1LimitDegrees=45
  swing1Motion=Free
  swing2LimitDegrees=45
  swing2Motion=Free
  twistLimitDegrees=45
  twistMotion=Free
 }
]
