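# Post-process style Animation Blueprint for the RimeSkate Plow female body, adding secondary motion to its torso, zipper and necklace.
# The copied parent-mesh pose passes through a Control Rig, weighted zipper constraint, then stacked angular and linear AnimDynamics.
# Two RigidBody nodes finish, crossfading between default and front-end (non-emote) physics assets.
# Updates derive torso blend alpha, crouch-turn flag and gravity override; an older torso dynamics chain sits disconnected.
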AnimBlueprint /BRCosmetics/Characters/Player/Female/Medium/Bodies/F_MED_RimeSkate_Plow/Meshes/F_MED_RimeSkate_Plow_AnimBP
target_skeleton=&/F_MED_RimeSkate_Plow_Skeleton
graphs [
 AnimGraph {nodes [
   n1 = "Transform (Modify) Bone - Bone: dyn_torso_l"(Alpha=1, Translation="0,0,1", Rotation="0,0,0", Scale="1,1,1") {translationMode=BMM_Additive, translationSpace=BCS_BoneSpace, boneToModify {boneName=dyn_torso_l}, translation {z=1}}
   n2 = "Transform (Modify) Bone - Bone: dyn_torso_r"(ComponentPose=@n1, Alpha=1, Translation="0,0,1", Rotation="0,0,0", Scale="1,1,1") {translationMode=BMM_Additive, translationSpace=BCS_BoneSpace, boneToModify {boneName=dyn_torso_r}, translation {z=1}}
   n3 = Anim Dynamics - Bone: dyn_torso_r!(ComponentPose=@n2, Alpha=1, GravityOverride:Vector="0,0,-980", ExternalForce="0,0,0") -> (Pose) {
    angularBiasOverride=0.999
    angularDampingOverride=0.999
    angularSpringConstant=4000
    bAngularSpring=true
    bGravityOverrideInSimSpace=true
    bOverrideAngularBias=true
    bOverrideAngularDamping=true
    bUseGravityOverride=true
    linearDampingOverride=0.9999
    linearSpringConstant=100
    simulationSpace=RootRelative
    boundBone {boneName=dyn_torso_r}
    gravityOverride {z=-980}
    physicsBodyDefinitions [{
      boundBone {boneName=dyn_torso_r}
      boxExtents=<21 21 21>
      constraintSetup {angularLimitsMax=<0 7.5 7.5>, angularLimitsMin=<0 -17.5 -7.5>, angularTarget=<1 0 0>}
      localJointOffset=<2 0 0>
     }]
    relativeSpaceBone {boneName=spine_04}
    simSpaceSettings {maxAngularAcceleration=45, maxAngularVelocity=8, simSpaceAngularAlpha=0.4}
   }
   n4 = Anim Dynamics - Bone: dyn_torso_l!(ComponentPose=@n3, Alpha=1, GravityOverride:Vector="0,0,-980", ExternalForce="0,0,0") -> (Pose) {
    angularBiasOverride=0.999
    angularDampingOverride=0.999
    angularSpringConstant=4000
    bAngularSpring=true
    bGravityOverrideInSimSpace=true
    bOverrideAngularBias=true
    bOverrideAngularDamping=true
    bUseGravityOverride=true
    linearDampingOverride=0.9999
    linearSpringConstant=100
    simulationSpace=RootRelative
    boundBone {boneName=dyn_torso_l}
    gravityOverride {z=-980}
    physicsBodyDefinitions [{
      boundBone {boneName=dyn_torso_l}
      boxExtents=<20 20 20>
      constraintSetup {angularLimitsMax=<0 7.5 7.5>, angularLimitsMin=<0 -17.5 -7.5>, angularTarget=<1 0 0>}
      localJointOffset=<2 0 0>
     }]
    relativeSpaceBone {boneName=spine_04}
    simSpaceSettings {maxAngularAcceleration=45, maxAngularVelocity=8, simSpaceAngularAlpha=0.4}
   }
   n5 = Anim Dynamics - Bone: dyn_torso_r(ComponentPose=@n4, Alpha=1, GravityOverride="0,0,-200", ExternalForce="0,0,0") {
    angularBiasOverride=1
    angularDampingOverride=0.99
    angularSpringConstant=10000
    bGravityOverrideInSimSpace=true
    bLinearSpring=true
    bOverrideLinearDamping=true
    bUseGravityOverride=true
    linearDampingOverride=0.9999
    linearSpringConstant=180
    simulationSpace=RootRelative
    boundBone {boneName=dyn_torso_r}
    gravityOverride {z=-200}
    physicsBodyDefinitions [{
      boundBone {boneName=dyn_torso_r}
      boxExtents=<70 70 70>
      constraintSetup {angularLimitsMax=<0 0 0>, angularLimitsMin=<0 0 0>, angularTarget=<1 0 0>, linearAxesMax=<0 0.5 1>, linearAxesMin=<0 -0.5 -0.3>}
      localJointOffset=<0.5 0 0>
     }]
    relativeSpaceBone {boneName=spine_04}
    simSpaceSettings {maxAngularAcceleration=400, maxAngularVelocity=200, simSpaceAngularAlpha=0.1}
   }
   Anim Dynamics - Bone: dyn_torso_l(ComponentPose=@n5, Alpha=1, GravityOverride="0,0,-200", ExternalForce="0,0,0") {
    angularBiasOverride=1
    angularDampingOverride=0.99
    angularSpringConstant=10000
    bGravityOverrideInSimSpace=true
    bLinearSpring=true
    bOverrideLinearDamping=true
    bUseGravityOverride=true
    linearDampingOverride=0.9999
    linearSpringConstant=180
    simulationSpace=RootRelative
    boundBone {boneName=dyn_torso_l}
    gravityOverride {z=-200}
    physicsBodyDefinitions [{
      boundBone {boneName=dyn_torso_l}
      boxExtents=<80 80 80>
      constraintSetup {angularLimitsMax=<0 0 0>, angularLimitsMin=<0 0 0>, angularTarget=<1 0 0>, linearAxesMax=<0 0.5 1>, linearAxesMin=<0 -0.5 -0.3>}
      localJointOffset=<0.5 0 0>
     }]
    relativeSpaceBone {boneName=spine_04}
    simSpaceSettings {maxAngularAcceleration=400, maxAngularVelocity=200, simSpaceAngularAlpha=0.1}
   }
   n6 = Get PbBlendEnabled_remap_float() -> (PbBlendEnabled_remap_float:double)
   n7 = Control Rig!(
    Source=Copy Pose From Mesh() {bUseAttachedParent=true}
    Alpha
    bRigidBodyEnabled:bool=Get bRigidBodyEnabled()
    static_crouching_rotation:bool=Get static_crouching_rotation() -> (static_crouching_rotation:bool)
    Moving2D:bool=Get bIsMoving2D()
   ) -> (Pose) {controlRigAssetReference {blueprintRigClass=&/F_MED_RimeSkate_Plow_CtrlRig.F_MED_RimeSkate_Plow_CtrlRig_C}}
   n8 = "Constraint - BASE_spine_04_zipper (PPPPPPPPPP)"!(
    ComponentPose=Local To Component(LocalPose=@n7)
    Alpha=1
    ConstraintWeights_0=0
    ConstraintWeights_1:float=0.0003
    ConstraintWeights_2:float=0.3142
    ConstraintWeights_3:float=0.0002
    ConstraintWeights_4:float=0
    ConstraintWeights_5:float=0.0315
    ConstraintWeights_6:float=0.0196
    ConstraintWeights_7:float=0.6334
    ConstraintWeights_8:float=0.0008
    ConstraintWeights_9:float=0
   ) -> (Pose) {
    boneToModify {boneName=BASE_spine_04_zipper}
    constraintSetup [
     {targetBone {boneName=spine_02}}
     {targetBone {boneName=clavicle_l}}
     {targetBone {boneName=spine_05}}
     {targetBone {boneName=clavicle_r}}
     {targetBone {boneName=spine_03}}
     {targetBone {boneName=deform_pec_l}}
     {targetBone {boneName=deform_pec_r}}
     {targetBone {boneName=spine_04}}
     {targetBone {boneName=dyn_torso_l}}
     {targetBone {boneName=dyn_torso_r}}
    ]
    constraintWeights [1,1,1,1,1,1,1,1,1,1]
   }
   n9 = Anim Dynamics - Bone: zipper(ComponentPose=@n8, Alpha=1, ExternalForce="0,0,0") {
    angularDampingOverride=1
    angularSpringConstant=20000
    bAngularSpring=true
    bOverrideAngularDamping=true
    bUseGravityOverride=true
    gravityScale=0.35
    simulationSpace=RootRelative
    boundBone {boneName=zipper}
    componentLinearAccScale=<1 1 1>
    componentLinearVelScale=<1 1 1>
    gravityOverride {z=-980}
    physicsBodyDefinitions [{
      boundBone {boneName=zipper}
      boxExtents=<20 30 30>
      constraintSetup {angularLimitsMax=<15 15 30>, angularLimitsMin=<-15 -15 0>, angularTarget=<1 0 0>}
      localJointOffset=<4 0 0>
     }]
   }
   n10 = Anim Dynamics - Bone: necklace(ComponentPose=@n9, Alpha=1, ExternalForce="0,0,0") {
    angularDampingOverride=1
    angularSpringConstant=20000
    bAngularSpring=true
    bOverrideAngularDamping=true
    bUseGravityOverride=true
    gravityScale=0.35
    simulationSpace=RootRelative
    boundBone {boneName=necklace}
    componentLinearAccScale=<1 1 1>
    componentLinearVelScale=<1 1 1>
    gravityOverride {z=-980}
    physicsBodyDefinitions [{
      boundBone {boneName=necklace}
      boxExtents=<20 30 30>
      constraintSetup {angularLimitsMax=<15 15 30>, angularLimitsMin=<-15 -15 0>, angularTarget=<1 0 0>}
      localJointOffset=<4 0 0>
     }]
   }
   n11 = Anim Dynamics - Bone: necklace_end(ComponentPose=@n10, Alpha=1, ExternalForce="0,0,0") {
    angularDampingOverride=1
    angularSpringConstant=20000
    bAngularSpring=true
    bOverrideAngularDamping=true
    bUseGravityOverride=true
    gravityScale=0.35
    simulationSpace=RootRelative
    boundBone {boneName=necklace_end}
    componentLinearAccScale=<1 1 1>
    componentLinearVelScale=<1 1 1>
    gravityOverride {z=-980}
    physicsBodyDefinitions [{
      boundBone {boneName=necklace_end}
      boxExtents=<20 30 30>
      constraintSetup {angularLimitsMax=<15 15 15>, angularLimitsMin=<-15 -15 5>, angularTarget=<1 0 0>}
      localJointOffset=<4 0 0>
     }]
   }
   n12 = "Transform (Modify) Bone - Bone: dyn_torso_l"(ComponentPose=@n11, Alpha=1, Translation="0,0,1", Rotation="0,0,0", Scale="1,1,1") {translationMode=BMM_Additive, translationSpace=BCS_BoneSpace, boneToModify {boneName=dyn_torso_l}, translation {z=1}}
   n13 = "Transform (Modify) Bone - Bone: dyn_torso_r"(ComponentPose=@n12, Alpha=1, Translation="0,0,1", Rotation="0,0,0", Scale="1,1,1") {translationMode=BMM_Additive, translationSpace=BCS_BoneSpace, boneToModify {boneName=dyn_torso_r}, translation {z=1}}
   n14 = Get Sim Space Settings() -> (Sim Space Settings:Anim Phys Sim Space Settings Structure)
   n15 = Anim Dynamics - Bone: dyn_torso_r!(ComponentPose=@n13, Alpha=@n6, AngularSpringConstant:float=5000, SimSpaceSettings:Anim Phys Sim Space Settings Structure=@n14) -> (Pose) {
    angularBiasOverride=1
    angularDampingOverride=0.9999
    angularSpringConstant=5000
    bAngularSpring=true
    bGravityOverrideInSimSpace=true
    bOverrideAngularBias=true
    bOverrideAngularDamping=true
    bUseGravityOverride=true
    linearDampingOverride=0.9999
    linearSpringConstant=100
    simulationSpace=RootRelative
    boundBone {boneName=dyn_torso_r}
    componentAppliedLinearAccClamp=<2 2 2>
    gravityOverride {z=-980}
    physicsBodyDefinitions [{
      boundBone {boneName=dyn_torso_r}
      boxExtents=<21 21 21>
      constraintSetup {angularLimitsMax=<0 1 5>, angularLimitsMin=<0 -5 -5>, angularTarget=<1 0 0>}
      localJointOffset=<2 0 0>
     }]
    relativeSpaceBone {boneName=spine_04}
    simSpaceSettings {maxAngularAcceleration=45, maxAngularVelocity=8, simSpaceAngularAlpha=0.25}
   }
   n16 = Anim Dynamics - Bone: dyn_torso_l!(ComponentPose=@n15, Alpha=@n6, AngularSpringConstant:float=5000, SimSpaceSettings:Anim Phys Sim Space Settings Structure=@n14) -> (Pose) {
    angularBiasOverride=1
    angularDampingOverride=0.9999
    angularSpringConstant=5000
    bAngularSpring=true
    bGravityOverrideInSimSpace=true
    bOverrideAngularBias=true
    bOverrideAngularDamping=true
    bUseGravityOverride=true
    linearDampingOverride=0.9999
    linearSpringConstant=100
    simulationSpace=RootRelative
    boundBone {boneName=dyn_torso_l}
    componentAppliedLinearAccClamp=<2 2 2>
    gravityOverride {z=-980}
    physicsBodyDefinitions [{
      boundBone {boneName=dyn_torso_l}
      boxExtents=<20 20 20>
      constraintSetup {angularLimitsMax=<0 1 5>, angularLimitsMin=<0 -5 -5>, angularTarget=<1 0 0>}
      localJointOffset=<2 0 0>
     }]
    relativeSpaceBone {boneName=spine_04}
    simSpaceSettings {maxAngularAcceleration=45, maxAngularVelocity=8, simSpaceAngularAlpha=0.25}
   }
   n17 = Anim Dynamics - Bone: dyn_torso_r!(ComponentPose=@n16, Alpha=@n6, GravityOverride:Vector="0,0,-200", LinearSpringConstant:float=180) -> (Pose) {
    angularBiasOverride=1
    angularDampingOverride=0.99
    angularSpringConstant=10000
    bGravityOverrideInSimSpace=true
    bLinearSpring=true
    bOverrideAngularBias=true
    bOverrideLinearDamping=true
    bUseGravityOverride=true
    linearDampingOverride=1
    linearSpringConstant=180
    simulationSpace=RootRelative
    boundBone {boneName=dyn_torso_r}
    componentAppliedLinearAccClamp=<5 5 5>
    gravityOverride {z=-200}
    physicsBodyDefinitions [{
      boundBone {boneName=dyn_torso_r}
      boxExtents=<80 80 80>
      constraintSetup {angularLimitsMax=<0 0 0>, angularLimitsMin=<0 0 0>, angularTarget=<1 0 0>, linearAxesMax=<0 0.5 1>, linearAxesMin=<0 -0.5 -0.5>}
      localJointOffset=<0.2 0 0>
     }]
    relativeSpaceBone {boneName=spine_04}
    simSpaceSettings {maxAngularAcceleration=50, maxAngularVelocity=10, simSpaceAngularAlpha=0.1}
   }
   n18 = Anim Dynamics - Bone: dyn_torso_l!(ComponentPose=@n17, Alpha=@n6, GravityOverride:Vector="0,0,-200", LinearSpringConstant:float=180) -> (Pose) {
    angularBiasOverride=1
    angularDampingOverride=0.99
    angularSpringConstant=10000
    bGravityOverrideInSimSpace=true
    bLinearSpring=true
    bOverrideAngularBias=true
    bOverrideLinearDamping=true
    bUseGravityOverride=true
    linearDampingOverride=1
    linearSpringConstant=180
    simulationSpace=RootRelative
    boundBone {boneName=dyn_torso_l}
    componentAppliedLinearAccClamp=<5 5 5>
    gravityOverride {z=-200}
    physicsBodyDefinitions [{
      boundBone {boneName=dyn_torso_l}
      boxExtents=<90 90 90>
      constraintSetup {angularLimitsMax=<0 0 0>, angularLimitsMin=<0 0 0>, angularTarget=<1 0 0>, linearAxesMax=<0 0.5 1>, linearAxesMin=<0 -0.5 -0.5>}
      localJointOffset=<0.2 0 0>
     }]
    relativeSpaceBone {boneName=spine_04}
    simSpaceSettings {maxAngularAcceleration=50, maxAngularVelocity=10, simSpaceAngularAlpha=0.1}
   }
   n19 = Get FE() -> (FE:bool)
   n20 = RigidBody!(ComponentPose=@n18, bAlphaBoolEnabled:bool=NOT Boolean(A=@n19), OverrideWorldGravity:Vector="0,0,-980") -> (Pose) {
    alphaInputType=Bool
    bClampLinearTranslationLimitToRefPose=true
    bOverrideWorldGravity=true
    overridePhysicsAsset=&/F_MED_RimeSkate_Plow_Physics
    simulationSpace=BaseBoneSpace
    alphaBoolBlend {blendInTime=0.2, blendOption=Linear, blendOutTime=0.2}
    baseBoneRef {boneName=pelvis}
    overrideWorldGravity {z=-980}
    simSpaceSettings {
     maxAngularAcceleration=65
     maxAngularVelocity=1500
     maxLinearAcceleration=1000
     maxLinearVelocity=3000
     velocityScaleZ=0.3
     worldAlpha=1
     externalLinearDragV=<1 1 1>
    }
   }
   n21 = RigidBody(ComponentPose=@n20, bAlphaBoolEnabled=@n19, OverrideWorldGravity="0,0,-980") {
    alphaInputType=Bool
    bClampLinearTranslationLimitToRefPose=true
    bOverrideWorldGravity=true
    overridePhysicsAsset=&/F_MED_RimeSkate_Plow_Physics_FE
    simulationSpace=BaseBoneSpace
    alphaBoolBlend {blendInTime=0.2, blendOption=Linear, blendOutTime=0.2}
    baseBoneRef {boneName=pelvis}
    overrideWorldGravity {z=-980}
    simSpaceSettings {
     maxAngularAcceleration=65
     maxAngularVelocity=1500
     maxLinearAcceleration=1000
     maxLinearVelocity=3000
     velocityScaleZ=0.3
     worldAlpha=1
     externalLinearDragV=<1 1 1>
    }
   }
   Output Pose(Result=Component To Local(ComponentPose=@n21)) {layerGroup=None}
  ]}
 EventGraph {nodes [
   Event BlueprintUpdateAnimation()
   > n1 = "Gravity Override State Function (Parameter Asset) (Nativized)"(
    CharacterPartAnimInstance=Self-Reference()
    GameTimeInSeconds=GetGameTimeInSeconds()
    Parameters=&/F_MED_RimeSkate_Plow_Parameters
   )
   > Set Gravity Override(Gravity Override=@n1)
   n2 = Get bIsMoving2D()
   n3 = "Absolute (Float)"(A=BreakRotator(InRot=Get PawnDeltaRotation() -> (PawnDeltaRotation:Rotator)).Yaw)
   > Set static_crouching_rotation(static_crouching_rotation:bool=AND Boolean!(A="float > float"(A=@n3, B=2), B=Get bIsCrouching(), C:bool=NOT Boolean(A=@n2)) -> (ReturnValue)) -> (Output_Get:bool)
   n4 = Get bPbBlendEnabled() -> (bPbBlendEnabled:bool)
   > Set PbBlendEnabled_remap_float(PbBlendEnabled_remap_float:double=Remap Float(Input="To Float (Boolean)"(InBool=@n4), OldRange=(X=0,Y=1), NewRange=(X=0,Y=0.7)) -> (Output=~)) -> (Output_Get:double)
   n5 = Get bIsPlayingEmote()
   > Set FE(FE:bool=AND Boolean!(A=Get bIsInFrontEnd(), B=NOT Boolean(A=@n5)) -> (ReturnValue)) -> (Output_Get:bool)
  ]}
]
variables [
 FE {type=boolean}
 Gravity Override {type=Vector}
 PbBlendEnabled_remap_float {type=number}
 Sim Space Settings {
  type=AnimPhysSimSpaceSettings
  default {maxAngularAcceleration=60, maxAngularVelocity=12, simSpaceAngularAlpha=0.7, externalAngularVelocity=<0 0 0>}
 }
 static_crouching_rotation {type=boolean}
]
