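# Physics asset for the face accessory mesh of the female medium RimeSkate Plow character, setting up physics for five hair strands.
# Each strand is a four-capsule chain hung from the head with tight twist and swing limits, allowing a wider swing at the root.
# Neighboring strand tips are cross-linked by angular-free constraints, tying the chains into one row.
# Head, upper spine, clavicle and upper-arm twist capsules form the surrounding body shape, joined by angular-free constraints.
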PhysicsAsset /BRCosmetics/Characters/Player/Female/Medium/Bodies/F_MED_RimeSkate_Plow/Meshes/Parts/F_MED_RimeSkate_Plow_FaceAcc_Physics_v1
bodies {
 clavicle_l [{shapeType=Capsule, length=9.4996, radius=5.6582, shapeName=clavicle_l_capsule, center=<4.5361 -1.2979 -0.9233>, rotation=<rot 0 -90 88.305>}]
 clavicle_r [{shapeType=Capsule, length=9.4996, radius=5.6582, shapeName=clavicle_r_capsule, center=<-4.5361 1.2979 0.9233>, rotation=<rot 0 -90 -91.695>}]
 hair_L1_01 [{shapeType=Capsule, length=5.662, radius=2.5133, shapeName=hair_R1_01_capsule, center=<4.6524 0 -1.4063>, rotation=<rot 0 -90 -84.1292>}]
 hair_L1_03 [{
   shapeType=Capsule
   length=2.8522
   radius=2.8899
   shapeName=hair_R1_03_capsule
   center=<4.0463 -0.1443 -1.0806>
   rotation=<rot 0.562 -89.9674 -84.3466>
  }]
 hair_L1_04 [{shapeType=Capsule, length=2.7657, radius=2.8635, shapeName=hair_R1_04_capsule, center=<3.673 0 -1.3754>, rotation=<rot 90 0 0>}]
 hair_L1_05 [{
   shapeType=Capsule
   length=9.9117
   radius=2.9183
   shapeName=hair_R1_05_capsule
   center=<7.3122 -0.0459 -0.7194>
   rotation=<rot 78.5624 -168.2731 -172.2763>
  }]
 hair_L2_01 [{shapeType=Capsule, length=5.662, radius=2.5133, shapeName=hair_R1_01_capsule, center=<4.7536 0 -0.4216>, rotation=<rot 0 -90 -84.1292>}]
 hair_L2_03 [{
   shapeType=Capsule
   length=2.8522
   radius=2.8899
   shapeName=hair_R1_03_capsule
   center=<4.0463 -0.1443 -1.0806>
   rotation=<rot 0.562 -89.9674 -84.3466>
  }]
 hair_L2_04 [{shapeType=Capsule, length=2.7657, radius=2.8635, shapeName=hair_R1_04_capsule, center=<3.673 0 -1.3754>, rotation=<rot 90 0 0>}]
 hair_L2_05 [{
   shapeType=Capsule
   length=9.9117
   radius=2.9183
   shapeName=hair_R1_05_capsule
   center=<7.262 0.6448 -0.701>
   rotation=<rot 78.5624 -168.2731 -172.2763>
  }]
 hair_M_01 [{shapeType=Capsule, length=5.662, radius=2.5133, shapeName=hair_R1_01_capsule, center=<4.7536 0 -0.4216>, rotation=<rot 0 -90 -84.1292>}]
 hair_M_03 [{
   shapeType=Capsule
   length=2.8522
   radius=2.8899
   shapeName=hair_R1_03_capsule
   center=<4.0463 -0.1443 -1.0806>
   rotation=<rot 0.562 -89.9674 -84.3466>
  }]
 hair_M_04 [{shapeType=Capsule, length=2.7657, radius=2.8635, shapeName=hair_R1_04_capsule, center=<3.673 0 -1.3754>, rotation=<rot 90 0 0>}]
 hair_M_05 [{
   shapeType=Capsule
   length=9.9117
   radius=2.9183
   shapeName=hair_R1_05_capsule
   center=<7.2404 0.6649 -0.7044>
   rotation=<rot 78.5623 -168.2732 -173.9257>
  }]
 hair_R1_01 [{shapeType=Capsule, length=5.662, radius=2.5133, shapeName=hair_R1_01_capsule, center=<4.5933 0 -1.9812>, rotation=<rot 0 -90 -84.1292>}]
 hair_R1_03 [{
   shapeType=Capsule
   length=2.8522
   radius=2.8899
   shapeName=hair_R1_03_capsule
   center=<4.0245 -0.3797 -1.2999>
   rotation=<rot 0.0019 -94.8227 -96.5887>
  }]
 hair_R1_04 [{
   shapeType=Capsule
   length=3.4319
   radius=2.1973
   shapeName=hair_R1_04_capsule
   center=<4.6963 -0.6284 -0.937>
   rotation=<rot 0.2843 -92.8817 -98.0862>
  }]
 hair_R1_05 [{
   shapeType=Capsule
   length=9.9117
   radius=2.9183
   shapeName=hair_R1_05_capsule
   center=<7.3122 -0.0459 -0.7194>
   rotation=<rot 78.5624 -168.2731 -172.2763>
  }]
 hair_R2_01 [{shapeType=Capsule, length=5.662, radius=2.5133, shapeName=hair_R2_01_capsule, center=<4.3316 0 -0.4454>, rotation=<rot 0 -90 -84.1292>}]
 hair_R2_03 [{
   shapeType=Capsule
   length=2.8522
   radius=2.8899
   shapeName=hair_R1_03_capsule
   center=<4.0463 -0.1443 -1.0806>
   rotation=<rot 0.562 -89.9674 -84.3466>
  }]
 hair_R2_04 [{shapeType=Capsule, length=2.7657, radius=2.8635, shapeName=hair_R1_04_capsule, center=<3.673 0 -1.3754>, rotation=<rot 90 0 0>}]
 hair_R2_05 [{
   shapeType=Capsule
   length=9.9117
   radius=2.9183
   shapeName=hair_R1_05_capsule
   center=<7.3122 -0.0459 -0.7194>
   rotation=<rot 78.5624 -168.2731 -172.2763>
  }]
 head [{shapeType=Capsule, length=6.7534, radius=6.3369, center=<3.1734 2.6019 0.0541>, rotation=<rot 0.5016 -90.415 -88.3249>}]
 spine_05 [{shapeType=Capsule, length=6.7148, radius=15.696, shapeName=spine_05_capsule, center=<0.9705 6.2334 0>, rotation=<rot 0 -90 0>}]
 upperarm_twist_01_l [{shapeType=Capsule, length=13.9458, radius=4.8773, shapeName=clavicle_l_capsule, center=<2.2719 -1.2979 0.6337>, rotation=<rot 0 -90 88.305>}]
 upperarm_twist_01_r [{
   shapeType=Capsule
   length=13.9458
   radius=4.8773
   shapeName=clavicle_r_capsule
   center=<-2.2719 1.2979 -0.6337>
   rotation=<rot 0 -90 -91.695>
  }]
}
constraints [
 {bone1Name=hair_L1_01, bone2Name=head, swing1LimitDegrees=10, swing2LimitDegrees=50, twistLimitDegrees=5}
 {bone1Name=hair_L2_01, bone2Name=head, swing1LimitDegrees=10, swing2LimitDegrees=50, twistLimitDegrees=5}
 {bone1Name=hair_M_01, bone2Name=head, swing1LimitDegrees=10, swing2LimitDegrees=50, twistLimitDegrees=5}
 {bone1Name=hair_R1_01, bone2Name=head, swing1LimitDegrees=10, swing2LimitDegrees=50, twistLimitDegrees=5}
 {bone1Name=hair_R2_01, bone2Name=head, swing1LimitDegrees=10, swing2LimitDegrees=50, twistLimitDegrees=5}
 {bone1Name=hair_L1_03, bone2Name=hair_L1_01, swing1LimitDegrees=10, swing2LimitDegrees=20, twistLimitDegrees=5}
 {bone1Name=hair_L1_04, bone2Name=hair_L1_03, swing1LimitDegrees=10, swing2LimitDegrees=20, twistLimitDegrees=5}
 {bone1Name=hair_L1_05, bone2Name=hair_L1_04, swing1LimitDegrees=10, swing2LimitDegrees=20, twistLimitDegrees=5}
 {bone1Name=hair_L2_03, bone2Name=hair_L2_01, swing1LimitDegrees=10, swing2LimitDegrees=20, twistLimitDegrees=5}
 {bone1Name=hair_L2_04, bone2Name=hair_L2_03, swing1LimitDegrees=10, swing2LimitDegrees=20, twistLimitDegrees=5}
 {bone1Name=hair_L2_05, bone2Name=hair_L2_04, swing1LimitDegrees=10, swing2LimitDegrees=20, twistLimitDegrees=5}
 {bone1Name=hair_M_03, bone2Name=hair_M_01, swing1LimitDegrees=10, swing2LimitDegrees=20, twistLimitDegrees=5}
 {bone1Name=hair_M_04, bone2Name=hair_M_03, swing1LimitDegrees=10, swing2LimitDegrees=20, twistLimitDegrees=5}
 {bone1Name=hair_M_05, bone2Name=hair_M_04, swing1LimitDegrees=10, swing2LimitDegrees=20, twistLimitDegrees=5}
 {bone1Name=hair_R1_04, bone2Name=hair_R1_03, swing1LimitDegrees=10, swing2LimitDegrees=20, twistLimitDegrees=5}
 {bone1Name=hair_R1_03, bone2Name=hair_R1_01, swing1LimitDegrees=10, swing2LimitDegrees=20, twistLimitDegrees=5}
 {bone1Name=hair_R1_05, bone2Name=hair_R1_04, swing1LimitDegrees=10, swing2LimitDegrees=20, twistLimitDegrees=5}
 {
  bone1Name=head
  bone2Name=spine_05
  swing1LimitDegrees=45
  swing1Motion=Free
  swing2LimitDegrees=45
  swing2Motion=Free
  twistLimitDegrees=45
  twistMotion=Free
 }
 {bone1Name=hair_R2_03, bone2Name=hair_R2_01, swing1LimitDegrees=10, swing2LimitDegrees=20, twistLimitDegrees=5}
 {bone1Name=hair_R2_04, bone2Name=hair_R2_03, swing1LimitDegrees=10, swing2LimitDegrees=20, twistLimitDegrees=5}
 {bone1Name=hair_R2_05, bone2Name=hair_R2_04, swing1LimitDegrees=10, swing2LimitDegrees=20, twistLimitDegrees=5}
 {
  bone1Name=hair_L2_05
  bone2Name=hair_L1_05
  swing1LimitDegrees=45
  swing1Motion=Free
  swing2LimitDegrees=45
  swing2Motion=Free
  twistLimitDegrees=45
  twistMotion=Free
 }
 {
  bone1Name=hair_M_05
  bone2Name=hair_L2_05
  swing1LimitDegrees=45
  swing1Motion=Free
  swing2LimitDegrees=45
  swing2Motion=Free
  twistLimitDegrees=45
  twistMotion=Free
 }
 {
  bone1Name=hair_M_05
  bone2Name=hair_R2_05
  swing1LimitDegrees=45
  swing1Motion=Free
  swing2LimitDegrees=45
  swing2Motion=Free
  twistLimitDegrees=45
  twistMotion=Free
 }
 {
  bone1Name=hair_R2_05
  bone2Name=hair_R1_05
  swing1LimitDegrees=45
  swing1Motion=Free
  swing2LimitDegrees=45
  swing2Motion=Free
  twistLimitDegrees=45
  twistMotion=Free
 }
 {
  bone1Name=clavicle_l
  bone2Name=spine_05
  swing1LimitDegrees=45
  swing1Motion=Free
  swing2LimitDegrees=45
  swing2Motion=Free
  twistLimitDegrees=45
  twistMotion=Free
 }
 {
  bone1Name=upperarm_twist_01_l
  bone2Name=clavicle_l
  swing1LimitDegrees=45
  swing1Motion=Free
  swing2LimitDegrees=45
  swing2Motion=Free
  twistLimitDegrees=45
  twistMotion=Free
 }
 {
  bone1Name=clavicle_r
  bone2Name=spine_05
  swing1LimitDegrees=45
  swing1Motion=Free
  swing2LimitDegrees=45
  swing2Motion=Free
  twistLimitDegrees=45
  twistMotion=Free
 }
 {
  bone1Name=upperarm_twist_01_r
  bone2Name=clavicle_r
  swing1LimitDegrees=45
  swing1Motion=Free
  swing2LimitDegrees=45
  swing2Motion=Free
  twistLimitDegrees=45
  twistMotion=Free
 }
]
