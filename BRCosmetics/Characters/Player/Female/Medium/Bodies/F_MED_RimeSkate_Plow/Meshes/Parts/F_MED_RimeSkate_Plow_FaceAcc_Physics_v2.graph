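# Physics asset for the FaceAcc mesh of the female RimeSkate Plow character, covering secondary motion of hair, locks, strands and earrings.
# Capsule bodies chain from the head through three hair columns, two side locks, short strands and earrings, with angular limits per joint.
# Cross constraints tie the left and right hair columns to the middle one below the roots.
# Unconstrained spine, clavicle and upper-arm capsules provide body collision shapes.
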PhysicsAsset /BRCosmetics/Characters/Player/Female/Medium/Bodies/F_MED_RimeSkate_Plow/Meshes/Parts/F_MED_RimeSkate_Plow_FaceAcc_Physics_v2
bodies {
 clavicle_l [{shapeType=Capsule, length=9.4996, radius=5.6582, shapeName=clavicle_l_capsule, center=<4.5361 -1.2979 -0.9233>, rotation=<rot 0 -90 88.305>}]
 clavicle_r [{shapeType=Capsule, length=9.4996, radius=5.6582, shapeName=clavicle_r_capsule, center=<-4.5361 1.2979 0.9233>, rotation=<rot 0 -90 -91.695>}]
 earring_l [{
   shapeType=Capsule
   length=1.8818
   radius=1.098
   shapeName=earring_l_capsule
   center=<2.0959 0.0103 -0.014>
   rotation=<rot 0 -89.3291 -89.6062>
  }]
 earring_r [{
   shapeType=Capsule
   length=1.8818
   radius=1.098
   shapeName=earring_l_capsule
   center=<2.0443 0.0097 -0.0136>
   rotation=<rot 0 -89.3291 -89.6062>
  }]
 hair_L1_01 [{shapeType=Capsule, length=5.662, radius=2.5133, shapeName=hair_R1_01_capsule, center=<4.6524 0 -1.4063>, rotation=<rot 0 -90 -84.1292>}]
 hair_L1_03 [{
   shapeType=Capsule
   length=3.9142
   radius=4.2194
   shapeName=hair_R1_03_capsule
   center=<3.7588 0.0279 -3.9228>
   rotation=<rot 5.6499 -177.9425 -86.3442>
  }]
 hair_L1_04 [{shapeType=Capsule, length=4.8335, radius=3.976, shapeName=hair_R1_04_capsule, center=<3.673 0 -3.1818>, rotation=<rot 90 91.1851 0>}]
 hair_L1_05 [{
   shapeType=Capsule
   length=7.0544
   radius=5.7756
   shapeName=hair_R1_05_capsule
   center=<7.8046 0.0563 -3.2053>
   rotation=<rot 78.5624 -168.2731 -172.2763>
  }]
 hair_M_01 [{shapeType=Capsule, length=5.662, radius=2.5133, shapeName=hair_R1_01_capsule, center=<4.7536 0 -0.4216>, rotation=<rot 0 -90 -84.1292>}]
 hair_M_03 [{
   shapeType=Capsule
   length=5.8473
   radius=4.0143
   shapeName=hair_R1_03_capsule
   center=<4.0463 -0.1443 -1.0806>
   rotation=<rot 5.6365 178.4953 -90.7105>
  }]
 hair_M_04 [{
   shapeType=Capsule
   length=6.563
   radius=3.7932
   shapeName=hair_R1_04_capsule
   center=<3.6725 0.0253 -1.818>
   rotation=<rot 88.4903 -88.8175 179.9974>
  }]
 hair_M_05 [{
   shapeType=Capsule
   length=6.635
   radius=6.1951
   shapeName=hair_R1_05_capsule
   center=<7.7648 0.7737 -3.3519>
   rotation=<rot 78.5623 -168.2732 -173.9257>
  }]
 hair_R1_01 [{shapeType=Capsule, length=5.662, radius=2.5133, shapeName=hair_R1_01_capsule, center=<4.5933 0 -1.9812>, rotation=<rot 0 -90 -84.1292>}]
 hair_R1_03 [{
   shapeType=Capsule
   length=2.5805
   radius=3.8639
   shapeName=hair_R1_03_capsule
   center=<4.1376 -0.6597 -2.3886>
   rotation=<rot 6.5837 -2.5965 -75.9351>
  }]
 hair_R1_04 [{
   shapeType=Capsule
   length=3.5234
   radius=3.5112
   shapeName=hair_R1_04_capsule
   center=<4.9358 -1.0482 -2.5299>
   rotation=<rot 73.1202 119.7103 33.7003>
  }]
 hair_R1_05 [{
   shapeType=Capsule
   length=7.4853
   radius=5.3448
   shapeName=hair_R1_05_capsule
   center=<7.6604 0.0264 -2.4774>
   rotation=<rot 78.5624 -168.2731 -172.2763>
  }]
 head [{shapeType=Capsule, length=6.7534, radius=6.3369, center=<3.1734 2.6019 0.0541>, rotation=<rot 0.5016 -90.415 -88.3249>}]
 lock_L_01 [{
   shapeType=Capsule
   length=3.5024
   radius=2.89
   shapeName=lock_R_01_capsule
   center=<3.3771 -2.602 0.1137>
   rotation=<rot 2.2542 -90 177.4805>
  }]
 lock_L_02 [{
   shapeType=Capsule
   length=3.5024
   radius=2.89
   shapeName=lock_R_01_capsule
   center=<3.3095 -2.7029 0.1147>
   rotation=<rot 2.8385 -74.7132 178.1637>
  }]
 lock_L_03 [{
   shapeType=Capsule
   length=2.7273
   radius=1.6813
   shapeName=lock_R_03_capsule
   center=<2.6579 -1.0169 -0.2244>
   rotation=<rot 1.1491 79.7788 -87.7119>
  }]
 lock_R_01 [{
   shapeType=Capsule
   length=3.5024
   radius=2.89
   shapeName=lock_R_01_capsule
   center=<3.3771 2.3573 -0.0815>
   rotation=<rot 2.2542 -90 177.4805>
  }]
 lock_R_02 [{
   shapeType=Capsule
   length=3.5024
   radius=2.89
   shapeName=lock_R_01_capsule
   center=<3.2863 2.6352 -0.0965>
   rotation=<rot 2.8385 -74.7132 178.1637>
  }]
 lock_R_03 [{
   shapeType=Capsule
   length=2.7273
   radius=1.6813
   shapeName=lock_R_03_capsule
   center=<3.2075 2.0313 -0.1623>
   rotation=<rot 0 106.4309 -87.4398>
  }]
 spine_03 [{shapeType=Capsule, length=15.5082, radius=10.4367, shapeName=spine_03_capsule, center=<2.8404 0.7671 0>, rotation=<rot 0 -90 0>}]
 spine_05 [{shapeType=Capsule, length=6.7148, radius=15.696, shapeName=spine_05_capsule, center=<0.9705 6.2334 0>, rotation=<rot 0 -90 0>}]
 strand_a_01 [{
   shapeType=Capsule
   length=2.2045
   radius=0.8612
   shapeName=strand_a_01_capsule
   center=<1.8739 -0.0164 0.1244>
   rotation=<rot 0.1758 91.5933 -106.6765>
  }]
 strand_a_02 [{
   shapeType=Capsule
   length=0.9606
   radius=1.2369
   shapeName=strand_a_02_capsule
   center=<2.2002 0.0463 0.261>
   rotation=<rot 0 -89.1682 -87.3885>
  }]
 strand_b_01 [{
   shapeType=Capsule
   length=1.2761
   radius=0.8924
   shapeName=strand_b_01_capsule
   center=<1.6597 -0.0168 0.879>
   rotation=<rot 0 -90.9657 -88.9242>
  }]
 strand_c_01 [{
   shapeType=Capsule
   length=0.505
   radius=0.505
   shapeName=strand_c_01_capsule
   center=<2.0273 -0.051 -0.1477>
   rotation=<rot 0 -92.3966 -82.996>
  }]
 upperarm_twist_01_l [{shapeType=Capsule, length=13.9458, radius=4.8773, shapeName=clavicle_l_capsule, center=<2.2719 -1.2979 0.6337>, rotation=<rot 0 -90 88.305>}]
 upperarm_twist_01_r [{
   shapeType=Capsule
   length=13.9458
   radius=4.8773
   shapeName=clavicle_r_capsule
   center=<-2.2719 1.2979 -0.6337>
   rotation=<rot 0 -90 -91.695>
  }]
}
constraints [
 {bone1Name=hair_L1_01, bone2Name=head, swing1LimitDegrees=10, swing2LimitDegrees=50, twistLimitDegrees=5}
 {bone1Name=hair_M_01, bone2Name=head, swing1LimitDegrees=10, swing2LimitDegrees=50, twistLimitDegrees=5}
 {bone1Name=hair_R1_01, bone2Name=head, swing1LimitDegrees=10, swing2LimitDegrees=50, twistLimitDegrees=5}
 {bone1Name=hair_L1_03, bone2Name=hair_L1_01, swing1LimitDegrees=10, swing2LimitDegrees=40, twistLimitDegrees=5}
 {bone1Name=hair_L1_04, bone2Name=hair_L1_03, swing1LimitDegrees=10, swing2LimitDegrees=20, twistLimitDegrees=5}
 {bone1Name=hair_L1_05, bone2Name=hair_L1_04, swing1LimitDegrees=10, swing2LimitDegrees=20, twistLimitDegrees=5}
 {bone1Name=hair_M_03, bone2Name=hair_M_01, swing1LimitDegrees=10, swing2LimitDegrees=40, twistLimitDegrees=5}
 {bone1Name=hair_M_04, bone2Name=hair_M_03, swing1LimitDegrees=10, swing2LimitDegrees=20, twistLimitDegrees=5}
 {bone1Name=hair_M_05, bone2Name=hair_M_04, swing1LimitDegrees=10, swing2LimitDegrees=20, twistLimitDegrees=5}
 {bone1Name=hair_R1_04, bone2Name=hair_R1_03, swing1LimitDegrees=10, swing2LimitDegrees=20, twistLimitDegrees=5}
 {bone1Name=hair_R1_03, bone2Name=hair_R1_01, swing1LimitDegrees=10, swing2LimitDegrees=40, twistLimitDegrees=5}
 {bone1Name=hair_R1_05, bone2Name=hair_R1_04, swing1LimitDegrees=10, swing2LimitDegrees=20, twistLimitDegrees=5}
 {bone1Name=hair_R1_05, bone2Name=hair_M_05, swing1LimitDegrees=30, swing2LimitDegrees=15, twistLimitDegrees=15}
 {bone1Name=hair_L1_05, bone2Name=hair_M_05, swing1LimitDegrees=30, swing2LimitDegrees=15, twistLimitDegrees=15}
 {bone1Name=hair_R1_04, bone2Name=hair_M_04, swing1LimitDegrees=30, swing2LimitDegrees=15, twistLimitDegrees=15}
 {bone1Name=hair_L1_04, bone2Name=hair_M_04, swing1LimitDegrees=30, swing2LimitDegrees=15, twistLimitDegrees=15}
 {bone1Name=lock_L_01, bone2Name=head, swing1LimitDegrees=10, swing2LimitDegrees=15, twistLimitDegrees=5}
 {bone1Name=lock_L_02, bone2Name=lock_L_01, swing1LimitDegrees=10, swing2LimitDegrees=15, twistLimitDegrees=5}
 {bone1Name=lock_L_03, bone2Name=lock_L_02, swing1LimitDegrees=10, swing2LimitDegrees=15, twistLimitDegrees=5}
 {bone1Name=lock_R_01, bone2Name=head, swing1LimitDegrees=10, swing2LimitDegrees=20, twistLimitDegrees=5}
 {bone1Name=lock_R_02, bone2Name=lock_R_01, swing1LimitDegrees=10, swing2LimitDegrees=20, twistLimitDegrees=5}
 {bone1Name=lock_R_03, bone2Name=lock_R_02, swing1LimitDegrees=10, swing2LimitDegrees=20, twistLimitDegrees=5}
 {bone1Name=strand_a_01, bone2Name=head, swing1LimitDegrees=5, swing2LimitDegrees=40, twistLimitDegrees=5}
 {bone1Name=strand_a_02, bone2Name=strand_a_01, swing1LimitDegrees=10, swing2LimitDegrees=40, twistLimitDegrees=5}
 {bone1Name=strand_b_01, bone2Name=head, swing1LimitDegrees=10, swing2LimitDegrees=30, twistLimitDegrees=10}
 {bone1Name=strand_c_01, bone2Name=head, swing1LimitDegrees=40, swing2LimitDegrees=30, twistLimitDegrees=10}
 {bone1Name=earring_l, bone2Name=head, swing1LimitDegrees=50, swing2LimitDegrees=5, twistLimitDegrees=5}
 {bone1Name=earring_r, bone2Name=head, swing1LimitDegrees=50, swing2LimitDegrees=5, twistLimitDegrees=5}
 {bone1Name=hair_R1_03, bone2Name=hair_M_03, swing1LimitDegrees=30, swing2LimitDegrees=15, twistLimitDegrees=15}
 {bone1Name=hair_L1_03, bone2Name=hair_M_03, swing1LimitDegrees=30, swing2LimitDegrees=15, twistLimitDegrees=15}
]
