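# Pre-physics Control Rig for the face accessory part of the female medium RimeSkate Plow character cosmetic.
# Its hierarchy holds the upper-body and head skeleton, with facial bones, five hair chains, side locks, strands, earrings and upper-arm deform bones.
# The graph contains only an unconnected Begin Execution node and the function library is empty, so the rig applies no adjustments.
# Two backpack cape modifier variables, a bool and a float, are declared but unused.
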ControlRigBlueprint /BRCosmetics/Characters/Player/Female/Medium/Bodies/F_MED_RimeSkate_Plow/Meshes/Parts/F_MED_RimeSkate_Plow_FaceAcc_PrePhysics_CtrlRig
graphs [
 RigVMFunctionLibrary {links [], nodes []}
 RigVMModel {nodes [RigUnit_BeginExecution = &RigUnit_BeginExecution()]}
]
hierarchy [
 root {loc=<0 0 0>, rot=<0 0 0 1>}
 pelvis {parent=root, loc=<0 -0.5301 98.0571>, rot=<0.0127 0.707 -0.0127 -0.707>}
 spine_01 {parent=pelvis, loc=<6.1208 -0.9328 0>, rot=<0 0 0.034 -0.9994>}
 spine_02 {parent=spine_01, loc=<6.5997 -0.2095 0>, rot=<0 0 0.0238 -0.9997>}
 spine_03 {parent=spine_02, loc=<6.4914 -0.1237 0>, rot=<0 0 0.0149 -0.9999>}
 spine_04 {parent=spine_03, loc=<6.9598 0.0794 0>, rot=<0 0 -0.0132 -0.9999>}
 spine_05 {parent=spine_04, loc=<6.9987 0 0>, rot=<0 0 -0.0015 -1>}
 neck_01 {parent=spine_05, loc=<11.3138 0.1529 0>, rot=<0 0 -0.2003 -0.9797>}
 neck_02 {parent=neck_01, loc=<4.5627 0 0>, rot=<0 0 0.0504 -0.9987>}
 head {parent=neck_02, loc=<5.3946 -0.0007 -0.0012>, rot=<0 0.0002 0.1145 -0.9934>}
 hair_L1_01 {parent=head, loc=<4.1073 -3.7482 -4.1059>, rot=<-0.0677 0.5148 0.8546 -0.0054>}
 hair_L1_02 {parent=hair_L1_01, loc=<3.4065 0 0>, rot=<0 0 0 1>}
 hair_L1_03 {parent=hair_L1_02, loc=<6.4688 0 0>, rot=<0 -0.3146 0 -0.9492>}
 hair_L1_04 {parent=hair_L1_03, loc=<6.4688 0 0>, rot=<0 0.1502 0 -0.9887>}
 hair_L1_05 {parent=hair_L1_04, loc=<6.4688 0 0>, rot=<0 0.1746 0 -0.9846>}
 hair_R1_01 {parent=head, loc=<3.9503 -3.5005 3.7489>, rot=<0.0156 -0.8643 -0.4937 0.0955>}
 hair_R1_02 {parent=hair_R1_01, loc=<3.4065 0 0>, rot=<0 0 0 1>}
 hair_R1_03 {parent=hair_R1_02, loc=<5.5304 0 0>, rot=<-0.0648 -0.2679 0.0442 -0.9603>}
 hair_R1_04 {parent=hair_R1_03, loc=<6.3656 0 0>, rot=<0.029 0.1096 -0.0012 -0.9936>}
 hair_R1_05 {parent=hair_R1_04, loc=<7.9626 0 0>, rot=<0 0.2157 0 -0.9765>}
 hair_M_01 {parent=head, loc=<3.9056 -6.8385 0.3428>, rot=<-0.1094 0.6866 0.7112 -0.1038>}
 hair_M_02 {parent=hair_M_01, loc=<3.4065 0 0>, rot=<0 0 0 1>}
 hair_M_03 {parent=hair_M_02, loc=<4.6411 0 0>, rot=<0 0 0 1>}
 hair_M_04 {parent=hair_M_03, loc=<6.4688 0 0>, rot=<0 0 0 1>}
 hair_M_05 {parent=hair_M_04, loc=<6.4688 0 0>, rot=<0.0033 0.1446 0.0559 -0.9879>}
 FACIAL_C_FacialRoot {parent=head, loc=<2.1524 -0.746 -0.0002>, rot=<0.4879 -0.5118 -0.4879 -0.5118>}
 FACIAL_L_ForeheadOut {parent=FACIAL_C_FacialRoot, loc=<4.3225 -2.9259 6.6357>, rot=<-0.1253 -0.2552 -0.0801 -0.9554>}
 FACIAL_R_ForeheadOut {parent=FACIAL_C_FacialRoot, loc=<-4.3217 -2.9246 6.636>, rot=<-0.1253 0.2553 0.0803 -0.9553>}
 FACIAL_L_Masseter {parent=FACIAL_C_FacialRoot, loc=<6.0778 1.6952 3.1075>, rot=<0.0653 0.1129 0.0084 -0.9914>}
 FACIAL_R_Masseter {parent=FACIAL_C_FacialRoot, loc=<-6.0806 1.6993 3.1037>, rot=<0.0653 -0.1128 -0.0083 -0.9914>}
 FACIAL_L_UnderChin {parent=FACIAL_C_FacialRoot, loc=<3.1378 7.0512 1.8743>, rot=<0.3991 -0.3552 0.0809 -0.8415>}
 FACIAL_R_UnderChin {parent=FACIAL_C_FacialRoot, loc=<-3.1387 7.0508 1.8762>, rot=<0.3991 0.3552 -0.0808 -0.8414>}
 FACIAL_C_Jaw {parent=FACIAL_C_FacialRoot, loc=<0.0004 2.5025 1.2252>, rot=<0.3382 0 0.0001 -0.9411>}
 FACIAL_L_Ear {parent=FACIAL_C_FacialRoot, loc=<6.2824 0.117 -0.4498>, rot=<-0.0304 -0.0005 -0.0261 -0.9992>}
 FACIAL_R_Ear {parent=FACIAL_C_FacialRoot, loc=<-6.2843 0.1181 -0.4566>, rot=<-0.0304 0.0006 0.0262 -0.9992>}
 FACIAL_L_Jawline {parent=FACIAL_C_FacialRoot, loc=<4.5819 5.0911 3.3454>, rot=<0.2054 0.158 -0.0539 -0.9643>}
 FACIAL_R_Jawline {parent=FACIAL_C_FacialRoot, loc=<-4.5874 5.0912 3.3456>, rot=<0.2054 -0.1579 0.0541 -0.9643>}
 lock_L_01 {parent=head, loc=<3.6901 1.8733 -6.4436>, rot=<0.0104 -0.707 -0.707 0.0146>}
 lock_L_02 {parent=lock_L_01, loc=<4.0206 0 0>, rot=<0.0027 0.0248 -0.1014 -0.9945>}
 lock_L_03 {parent=lock_L_02, loc=<3.8733 0 0>, rot=<0 0 0 1>}
 lock_R_01 {parent=head, loc=<3.2969 1.808 6.8232>, rot=<0.0434 -0.7795 -0.6243 -0.0279>}
 lock_R_02 {parent=lock_R_01, loc=<4.6799 0 0>, rot=<-0.0138 -0.0142 0.0928 -0.9955>}
 lock_R_03 {parent=lock_R_02, loc=<2.3226 0 0>, rot=<0 0 0 1>}
 earring_l {parent=head, loc=<0.5327 -0.846 -6.5156>, rot=<0.0038 -0.9813 -0.1925 -0.0005>}
 earring_r {parent=head, loc=<0.5327 -0.8449 6.5153>, rot=<0.0038 -0.9873 0.1589 0.0006>}
 strand_a_01 {parent=head, loc=<9.7348 7.7973 -2.5027>, rot=<-0.0674 -0.8725 -0.2691 -0.4022>}
 strand_a_02 {parent=strand_a_01, loc=<2.2564 0 0>, rot=<0 -0.3456 0 -0.9384>}
 strand_b_01 {parent=head, loc=<8.3685 7.0768 -4.0531>, rot=<0.2576 -0.8948 -0.2901 -0.2208>}
 strand_c_01 {parent=head, loc=<5.1708 5.0573 -6.0293>, rot=<0.0484 -0.9982 -0.0032 -0.0357>}
 hair_R2_01 {parent=head, loc=<3.7635 -5.4775 2.6236>, rot=<0.1173 -0.7742 -0.6002 0.163>}
 hair_R2_02 {parent=hair_R2_01, loc=<3.4065 0 0>, rot=<0 0 0 1>}
 hair_R2_03 {parent=hair_R2_02, loc=<4.6411 0 0>, rot=<0 0 0 1>}
 hair_R2_04 {parent=hair_R2_03, loc=<6.4688 0 0>, rot=<0 0 0 1>}
 hair_R2_05 {parent=hair_R2_04, loc=<6.4688 0 0>, rot=<0.0033 0.1446 0.0559 -0.9879>}
 hair_L2_01 {parent=head, loc=<3.8054 -5.996 -2.2806>, rot=<-0.1423 0.5762 0.7997 -0.0904>}
 hair_L2_02 {parent=hair_L2_01, loc=<3.4065 0 0>, rot=<0 0 0 1>}
 hair_L2_03 {parent=hair_L2_02, loc=<4.6411 0 0>, rot=<0 0 0 1>}
 hair_L2_04 {parent=hair_L2_03, loc=<6.4688 0 0>, rot=<0 0 0 1>}
 hair_L2_05 {parent=hair_L2_04, loc=<6.4688 0 0>, rot=<0.0033 0.1446 0.0559 -0.9879>}
 clavicle_l {parent=spine_05, loc=<5.4831 -0.1053 -3.2946>, rot=<0.0465 -0.7289 -0.009 -0.6829>}
 upperarm_l {parent=clavicle_l, loc=<9.8658 -1.2136 -0.4105>, rot=<0.0164 -0.3245 -0.0269 -0.9454>}
 upperarm_twist_01_l {parent=upperarm_l, loc=<7.2135 -0.1134 0>, rot=<0 0 0 1>}
 deform_upperarm_bk_base_l {parent=upperarm_twist_01_l, loc=<-6.4053 -4.3746 0.314>, rot=<0 0 0 -1>}
 deform_upperarm_bk_l {parent=deform_upperarm_bk_base_l, loc=<0 0 0>, rot=<0 0 0 1>}
 deform_upperarm_fr_base_l {parent=upperarm_twist_01_l, loc=<-5.6312 5.092 0.3354>, rot=<0 0 0 -1>}
 deform_upperarm_fr_l {parent=deform_upperarm_fr_base_l, loc=<0 0 0>, rot=<0 0 0 1>}
 deform_upperarm_up_base_l {parent=upperarm_twist_01_l, loc=<-8.891 0.4825 3.9517>, rot=<0 0 0 -1>}
 deform_upperarm_up_l {parent=deform_upperarm_up_base_l, loc=<0 0 0>, rot=<0 0 0 1>}
 clavicle_r {parent=spine_05, loc=<5.4831 -0.1053 3.2946>, rot=<0.7289 0.0465 -0.6829 0.009>}
 upperarm_r {parent=clavicle_r, loc=<-9.8658 1.2136 0.4105>, rot=<0.0164 -0.3245 -0.0269 -0.9454>}
 upperarm_twist_01_r {parent=upperarm_r, loc=<-7.2135 0.1134 0>, rot=<0 0 0 1>}
 deform_upperarm_bk_base_r {parent=upperarm_twist_01_r, loc=<6.4053 4.3746 -0.314>, rot=<0 0 0 -1>}
 deform_upperarm_bk_r {parent=deform_upperarm_bk_base_r, loc=<0 0 0>, rot=<0 0 0 1>}
 deform_upperarm_fr_base_r {parent=upperarm_twist_01_r, loc=<5.6312 -5.092 -0.3354>, rot=<0 0 0 -1>}
 deform_upperarm_fr_r {parent=deform_upperarm_fr_base_r, loc=<0 0 0>, rot=<0 0 0 1>}
 deform_upperarm_up_base_r {parent=upperarm_twist_01_r, loc=<8.891 -0.4825 -3.9517>, rot=<0 0 0 -1>}
 deform_upperarm_up_r {parent=deform_upperarm_up_base_r, loc=<0 0 0>, rot=<0 0 0 1>}
]
variables [
 backpack_cape_mod_bool {type=bool, default=False}
 backpack_cape_mod_float {type=double, default=0}
]
